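# Sugarcoat material layer function, filed under the Candy Girl glider, that adds sparkly sugar coating over a base material attributes input.
# A tiled gumdrop sparkle normal, flattened by Fresnel, and a sparkle mask are both scaled by the Sparkles parameter.
# The mask and Fresnel reshape base color, subsurface color, opacity, metallic, specular and roughness.
# A Mask input blends the result over the base, which must supply subsurface and base colors.
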MaterialFunction /BRCosmetics/Gadgets/Assets/VinderTech_GliderChute/Glider_CandyGirl/Material/MF_Sugarcoat
graph {
 properties {
  bExposeToLibrary=true
  description=You need SSS and Base colors on your base shader to make this part work correctly
  libraryCategoriesText [Material pass layer]
 }
 nodes [
  n1 = FunctionInput() {inputName=Base pass, inputType=FunctionInput_MaterialAttributes, previewValue=<0.5 0 0 1>}
  n2 = GetMaterialAttributes(None=@n1) {attributeGetTypes [#1]}
  n3 = Multiply(A=TextureCoordinate(), B=ScalarParameter() {defaultValue=2, group=Sugarcoat, parameterName=SparklesTiling})
  n4 = TextureSample(UVs=@n3) {
   samplerSource=SSM_Wrap_WorldGroupSettings
   samplerType=SAMPLERTYPE_Normal
   texture=/BRCosmetics/Gadgets/Assets/VinderTech_GliderChute/Textures/T_Gumdrop_Sparkle
  }
  n5 = Constant() {r=0.4}
  n6 = Constant() {r=-0.6242}
  n7 = VertexNormalWS()
  n8 = MaterialFunctionCall(
   BaseNormal=@n2.Normal
   AdditionalNormal=LinearInterpolate(
    A=@n4.RGB
    B=Constant3Vector() {constant=<rgba 0 0 1 0>}
    Alpha=Fresnel(ExponentIn=@n5, BaseReflectFractionIn=@n6, Normal=@n7) {baseReflectFraction=0.1, exponent=1}
   )
  ) {materialFunction=&BlendAngleCorrectedNormals}
  n9 = ScalarParameter() {defaultValue=1, group=Sugarcoat, parameterName=Sparkles}
  n10 = GetMaterialAttributes(None=SetMaterialAttributes!(MaterialAttributes=@n2.MaterialAttributes, Normal=LinearInterpolate(A=@n2.Normal, B=@n8.Result, Alpha=@n9)) {attributeSetTypes [#1]}) {attributeGetTypes [#2]}
  n11 = Constant() {r=1}
  n12 = Multiply(A=@n11, B=@n10.SubsurfaceColor)
  n13 = Constant() {r=0.3}
  n14 = Multiply(A=@n13, B=@n10.SubsurfaceColor)
  n15 = Fresnel(ExponentIn=Constant() {r=0.4}, BaseReflectFractionIn=Constant() {r=-0.6242}) {baseReflectFraction=0.1, exponent=1}
  n16 = GetMaterialAttributes(None=SetMaterialAttributes!(MaterialAttributes=@n10.MaterialAttributes, SubsurfaceColor=LinearInterpolate(A=@n12, B=@n14, Alpha=@n15)) {attributeSetTypes [#2]}) {attributeGetTypes [#3]}
  n17 = TextureSample(UVs=@n3) {
   samplerSource=SSM_Wrap_WorldGroupSettings
   samplerType=SAMPLERTYPE_Masks
   texture=/Game/Gadgets/Assets/VinderTech_GliderChute/Textures/T_Gumdrop_Sparkle_M
  }
  n18 = LinearInterpolate(B=@n17.G, Alpha=@n9)
  n19 = LinearInterpolate(A=Multiply(A=@n16.BaseColor) {constB=0.5}, B=@n16.BaseColor, Alpha=@n18)
  n20 = SetMaterialAttributes!(MaterialAttributes=@n16.MaterialAttributes, BaseColor=LinearInterpolate(A=Multiply(A=@n13, B=@n19), B=Multiply(A=@n11, B=@n19), Alpha=@n15)) {attributeSetTypes [#3]}
  n21 = LinearInterpolate(A=Constant() {r=0.25}, B=Constant() {r=-0.25}, Alpha=@n15)
  n22 = LinearInterpolate(A=Constant() {r=0.25}, B=Constant() {r=0.5}, Alpha=@n18)
  n23 = Constant() {r=-0.1429}
  n24 = GetMaterialAttributes(None=SetMaterialAttributes!(
    MaterialAttributes=SetMaterialAttributes!(MaterialAttributes=@n20, Opacity=@n21) {attributeSetTypes [#4]}
    Metallic=LinearInterpolate(A=@n22, B=@n23, Alpha=@n15)
   ) {attributeSetTypes [#5]}) {attributeGetTypes [#6]}
  n25 = LinearInterpolate(A=Constant() {r=0.5}, B=Constant() {r=1}, Alpha=@n18)
  n26 = GetMaterialAttributes(None=SetMaterialAttributes!(MaterialAttributes=@n24.MaterialAttributes, Specular=LinearInterpolate(A=@n25, B=@n24.Specular, Alpha=@n15)) {attributeSetTypes [#6]}) {attributeGetTypes [#7]}
  n27 = LinearInterpolate(A=LinearInterpolate(A=Constant() {r=0.4761}, B=Constant() {r=0.6124}, Alpha=@n18), B=@n26.Roughness, Alpha=@n15)
  FunctionOutput(None=BlendMaterialAttributes(
    A=@n1
    B=SetMaterialAttributes!(MaterialAttributes=@n26.MaterialAttributes, Roughness=@n27) {attributeSetTypes [#7]}
    Alpha=FunctionInput() {inputName=Mask, inputType=FunctionInput_Scalar, sortPriority=1}
   )) {description=You need SSS and Base colors on your base shader to make this part work correctly}
 ]
}
