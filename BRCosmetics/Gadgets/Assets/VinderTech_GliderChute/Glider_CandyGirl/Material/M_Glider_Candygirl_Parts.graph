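# Subsurface material for the parts of the Candy Girl glider cosmetic, built on material attributes for skeletal meshes.
# Diffuse feeds base and subsurface color, the S mask's green channel drives metallic, specular and roughness, and only mobile shading adds diffuse-based emissive.
# The E mask's green channel masks a MF_Sugarcoat pass, and its blue channel tints an Emissive parameter glow.
# The glow is layered through MatLayerBlend_Emissive, then passed through MF_TechArtOverrides_accessories.
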Material /BRCosmetics/Gadgets/Assets/VinderTech_GliderChute/Glider_CandyGirl/Material/M_Glider_Candygirl_Parts
graph {
 properties {
  bUseMaterialAttributes=true
  bUsedWithMeshDeformer=true
  bUsedWithSkeletalMesh=true
  pixelDepthOffsetMode=PDOM_Legacy
  shadingModel=MSM_Subsurface
  translucencyLightingMode=TLM_Surface
  usedShadingModels=Subsurface
 }
 nodes [
  n1 = TextureSample() {texture=/BRCosmetics/Gadgets/Assets/VinderTech_GliderChute/Glider_CandyGirl/Textures/Glider_Candy_Girl_Pasts_D}
  n2 = TextureSample() {
   samplerType=SAMPLERTYPE_Masks
   texture=/BRCosmetics/Gadgets/Assets/VinderTech_GliderChute/Glider_CandyGirl/Textures/Glider_Candy_Girl_Pasts_S
  }
  n3 = Constant()
  n4 = Multiply(A=@n1.RGB, B=ScalarParameter() {defaultValue=1, parameterName=ES3_Emissive}) {constB=0.05}
  n5 = TextureSample() {
   samplerType=SAMPLERTYPE_Normal
   texture=/BRCosmetics/Gadgets/Assets/VinderTech_GliderChute/Glider_CandyGirl/Textures/Glider_Candy_Girl_Pasts_N
  }
  n6 = MakeMaterialAttributes!(
   BaseColor=@n1.RGB
   Metallic=@n2.G
   Specular=@n2.G
   Roughness=@n2.G
   Anisotropy
   EmissiveColor=ShadingPathSwitch(Default=@n3, Deferred=@n3, Mobile=@n4) {desc=Migrated from FeatureLevelSwitch}
   Opacity=Constant() {r=1}
   OpacityMask
   Normal=@n5.RGB
   Tangent
   WorldPositionOffset
   SubsurfaceColor=@n1.RGB
   ClearCoat
   ClearCoatRoughness
   AmbientOcclusion
   Refraction
   CustomizedUVs_0
   CustomizedUVs_1
   CustomizedUVs_2
   CustomizedUVs_3
   CustomizedUVs_4
   CustomizedUVs_5
   CustomizedUVs_6
   CustomizedUVs_7
   PixelDepthOffset
   ShadingModel
   Displacement
  )
  n7 = TextureSample() {
   samplerType=SAMPLERTYPE_Masks
   texture=/BRCosmetics/Gadgets/Assets/VinderTech_GliderChute/Glider_CandyGirl/Textures/Glider_Candy_Girl_Pasts_E
  }
  n8 = VectorParameter() {parameterName=Emissive, defaultValue=<rgba 0.6042 4 4 1>}
  MaterialFunctionCall(In=MaterialFunctionCall!(
    Material=MaterialFunctionCall!(Base pass=@n6, Mask=@n7.G) {materialFunction=&/MF_Sugarcoat}.Result
    Emissive=Multiply(A=@n8.RGB, B=@n7.B)
   ) {materialFunction=/Engine/Functions/MaterialLayerFunctions/MatLayerBlend_Emissive}.Blended Material) {materialFunction=&MF_TechArtOverrides_accessories}
  QualitySwitch(Default=@n3, Low=@n4)
 ]
}
